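# Niagara effect for the DriedSilk emote that pops up an exclamation point in a bubble, then squashes it away.
# One burst emitter spawns a short-lived mesh particle tinted by time of day, with two mesh renderers.
# Curve-driven stacked scale modules animate the mesh, while custom ScaleExclamation and PositionExclamation attributes give the mark its own delayed pop, bounce and exit.
# Further curves animate dynamic material parameters over its lifetime.
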NiagaraSystem /BRCosmetics/Effects/Fort_Effects/Emotes/DriedSilk/NS_DriedSilk_Emote_ExclamationPoint
dependencies {usedDataInterfaces [&NiagaraDataInterfaceCurve], usedDynamicInputs [], usedModules [&SystemState]}
emitters {ExclamationBubble {
  emitterUpdateScript [
   &EmitterState()
   &SpawnBurst_Instantaneous(Use Spawn Probability=false, UseLoopCountLimit=false)
  ]
  particleSpawnScript [
   &InitializeParticle(
    Color=&FNTimeOfDayColor(
     Night Color=<rgba 0.6391 0.6868 0.83 1>
     Morning Color=<rgba 0.9 0.72 0.75 1>
     Day Color=<rgba 1 1 1 1>
     Evening Color=<rgba 0.85 0.7469 0.7395 1>
    )
    Lifetime=1.22
    Mesh Scale Mode=::NewEnumerator3
    Mesh Uniform Scale=0.15
    UsePositionOffset=false
    Write Material Random=true
   ) {module=InitializeParticle001}
   &SystemLocation(MaskSpawnGroup=false, Offset=<0 3.5 -12>)
  ]
  particleUpdateScript [
   &ParticleState()
   &DynamicMaterialParameters(
    Index 0 Param 1=&FloatFromCurve(FloatCurve=DataInterface() {propertyValues {Curve {keys [
         {arriveTangent=4.5008, interpMode=RCIM_Cubic, leaveTangent=4.5008, tangentMode=RCTM_User, time=0, value=0}
         {arriveTangent=0.0786, interpMode=RCIM_Cubic, leaveTangent=0.0786, tangentMode=RCTM_User, time=0.19, value=0.5}
         {arriveTangent=10.7793, interpMode=RCIM_Cubic, leaveTangent=-0.0903, tangentMode=RCTM_Break, time=0.25, value=1}
         {arriveTangent=0.0146, interpMode=RCIM_Cubic, leaveTangent=-0.4296, tangentMode=RCTM_Break, time=0.8, value=1}
        ]}}})
    Index 0 Param 2=1
    Index 0 Param 3=&FloatFromCurve(
     FloatCurve=DataInterface() {propertyValues {Curve {keys [
         {arriveTangent=-17.7515, interpMode=RCIM_Cubic, leaveTangent=-17.7515, tangentMode=RCTM_User, time=-0.0046, value=1.0113}
         {arriveTangent=-0.1143, interpMode=RCIM_Cubic, leaveTangent=-0.1143, tangentMode=RCTM_User, time=0.1, value=0.065}
         {arriveTangent=0.0901, interpMode=RCIM_Cubic, leaveTangent=0.0901, tangentMode=RCTM_User, time=0.7, value=0.065}
         {arriveTangent=3.5589, interpMode=RCIM_Cubic, leaveTangent=3.5589, tangentMode=RCTM_User, time=1, value=1}
        ]}}}
     Scale Curve=19
    )
    Module.Param0WriteEnabled=true
    Module.Param1WriteEnabled=false
    Module.Param2WriteEnabled=true
    Module.Param3WriteEnabled=true
   )
   &ScaleMeshSize(Scale Factor=&MakeVector(
     X=1
     Y=&FloatFromCurve(
      FloatCurve=DataInterface() {propertyValues {Curve {keys [
          <at 0 0 RCIM_Cubic>
          {arriveTangent=8.7331, interpMode=RCIM_Cubic, leaveTangent=8.7331, tangentMode=RCTM_User, time=0.1, value=0.35}
          {arriveTangent=61.6735, interpMode=RCIM_Cubic, leaveTangent=61.6735, tangentMode=RCTM_User, time=0.14, value=2.5}
          {arriveTangent=-0.1844, interpMode=RCIM_Cubic, leaveTangent=-0.1844, tangentMode=RCTM_User, time=0.19, value=0.3}
          {arriveTangent=0.4719, interpMode=RCIM_Cubic, leaveTangent=0.4719, tangentMode=RCTM_User, time=0.28, value=1.3}
          {arriveTangent=0.1261, interpMode=RCIM_Cubic, leaveTangent=0.1261, tangentMode=RCTM_User, time=0.33, value=1}
         ]}}}
      CurveIndex=$Particles.Age
     )
     Z=&FloatFromCurve(
      FloatCurve=DataInterface() {propertyValues {Curve {keys [
          {arriveTangent=35.0863, interpMode=RCIM_Cubic, leaveTangent=35.0863, tangentMode=RCTM_User, time=0, value=0}
          {arriveTangent=4.4969, interpMode=RCIM_Cubic, leaveTangent=4.4969, tangentMode=RCTM_User, time=0.14, value=0.5}
          {arriveTangent=21.0067, interpMode=RCIM_Cubic, leaveTangent=21.0067, tangentMode=RCTM_User, time=0.2, value=1.5}
          {arriveTangent=-0.1844, interpMode=RCIM_Cubic, leaveTangent=-0.1844, tangentMode=RCTM_User, time=0.26, value=0.9}
          <at 0.33 1 RCIM_Cubic>
         ]}}}
      CurveIndex=$Particles.Age
     )
    ))
   &ScaleMeshSize(Scale Factor=&MakeVector(
     X=1
     Y=&FloatFromCurve(FloatCurve=DataInterface() {propertyValues {Curve {keys [
          {arriveTangent=-1.9633, interpMode=RCIM_Cubic, leaveTangent=-1.9633, tangentMode=RCTM_User, time=0.86, value=1}
          {arriveTangent=-7.7161, interpMode=RCIM_Cubic, leaveTangent=-7.7161, tangentMode=RCTM_User, time=0.88, value=0.4}
          {arriveTangent=0.1065, interpMode=RCIM_Cubic, leaveTangent=0.1065, tangentMode=RCTM_User, time=0.99, value=0.16}
          {arriveTangent=-22.6812, interpMode=RCIM_Cubic, leaveTangent=-22.6812, tangentMode=RCTM_User, time=1, value=0}
         ]}}})
     Z=&FloatFromCurve(FloatCurve=DataInterface() {propertyValues {Curve {keys [
          {arriveTangent=0.1197, interpMode=RCIM_Cubic, leaveTangent=0.1197, tangentMode=RCTM_User, time=0.82, value=1}
          {arriveTangent=4, interpMode=RCIM_Cubic, leaveTangent=4, time=0.91, value=1.55}
          {arriveTangent=0.3586, interpMode=RCIM_Cubic, leaveTangent=0.3586, tangentMode=RCTM_User, time=0.97, value=1.6}
          {arriveTangent=-130.4235, interpMode=RCIM_Cubic, leaveTangent=-130.4235, tangentMode=RCTM_User, time=1, value=0}
         ]}}})
    )) {module=ScaleMeshSize002}
   /SetVariables(Particles.ScaleExclamation:Vector3f=&Multiply_VectorByFloat(
     Vector=&MakeVector(
      X=1
      Y=&FloatFromCurve(FloatCurve=DataInterface() {propertyValues {Curve {keys [
           {arriveTangent=15.4092, interpMode=RCIM_Cubic, leaveTangent=15.4092, tangentMode=RCTM_User, time=0.19, value=0}
           {arriveTangent=-4.8009, interpMode=RCIM_Cubic, leaveTangent=-4.8009, tangentMode=RCTM_User, time=0.23, value=1.6}
           {arriveTangent=-1.1737, interpMode=RCIM_Cubic, leaveTangent=-1.1737, tangentMode=RCTM_User, time=0.29, value=1}
          ]}}})
      Z=&FloatFromCurve(FloatCurve=DataInterface() {propertyValues {Curve {keys [
           {arriveTangent=52.0014, interpMode=RCIM_Cubic, leaveTangent=52.0014, tangentMode=RCTM_User, time=0.18, value=0}
           {arriveTangent=-4.2056, interpMode=RCIM_Cubic, leaveTangent=-4.2056, tangentMode=RCTM_User, time=0.2, value=1.2}
           {arriveTangent=-2.5, interpMode=RCIM_Cubic, leaveTangent=-2.5, time=0.22, value=0.8}
           {arriveTangent=-1.324, interpMode=RCIM_Cubic, leaveTangent=-1.324, tangentMode=RCTM_User, time=0.28, value=1}
          ]}}})
     )
     Float=0.15
    )) {isSetParameters=true}
   /SetVariables(Particles.ScaleExclamation:Vector3f=&Multiply_Vector(
     A=$Particles.ScaleExclamation
     B=&MakeVector(
      X=1
      Y=&FloatFromCurve(FloatCurve=DataInterface() {propertyValues {Curve {keys [
           <at 0.84 1 RCIM_Cubic>
           {arriveTangent=-6.3668, interpMode=RCIM_Cubic, leaveTangent=-6.3668, tangentMode=RCTM_User, time=0.92, value=0.3}
           {arriveTangent=-0.374, interpMode=RCIM_Cubic, leaveTangent=-0.374, tangentMode=RCTM_User, time=1, value=0}
          ]}}})
      Z=&FloatFromCurve(FloatCurve=DataInterface() {propertyValues {Curve {keys [
           <at 0.85 1 RCIM_Cubic>
           {arriveTangent=1.2631, interpMode=RCIM_Cubic, leaveTangent=1.2631, tangentMode=RCTM_User, time=0.87, value=1.5}
           {arriveTangent=-13.5073, interpMode=RCIM_Cubic, leaveTangent=-13.5073, tangentMode=RCTM_User, time=0.97, value=1.3}
           {arriveTangent=-4.713, interpMode=RCIM_Cubic, leaveTangent=-4.713, tangentMode=RCTM_User, time=1, value=0}
          ]}}})
     )
    )) {isSetParameters=true}
   /SetVariables(Particles.PositionExclamation:NiagaraPosition=&AddVectorToPosition(
     Position=$Particles.Position
     Vector=&AddVectorToPosition(
      Position=<0 0 30>
      Vector=&MakeVector(
       X=0
       Y=0
       Z=&FloatFromCurve(
        FloatCurve=DataInterface() {propertyValues {Curve {keys [
            {arriveTangent=39.4575, interpMode=RCIM_Cubic, leaveTangent=39.4575, tangentMode=RCTM_User, time=0.14, value=0}
            {arriveTangent=-28.8924, interpMode=RCIM_Cubic, leaveTangent=-28.8924, tangentMode=RCTM_User, time=0.18, value=1.3}
            {arriveTangent=0.2439, interpMode=RCIM_Cubic, leaveTangent=0.2439, tangentMode=RCTM_User, time=0.21, value=-0.3}
            {arriveTangent=5.8468, interpMode=RCIM_Cubic, leaveTangent=5.8468, tangentMode=RCTM_User, time=0.27, value=-0.1}
            {arriveTangent=-0.0152, interpMode=RCIM_Cubic, leaveTangent=-0.0152, tangentMode=RCTM_User, time=0.3, value=0}
            {arriveTangent=-0.0308, interpMode=RCIM_Cubic, leaveTangent=-0.0308, tangentMode=RCTM_User, time=0.82, value=0}
            {arriveTangent=-0.4117, interpMode=RCIM_Cubic, leaveTangent=-0.4117, tangentMode=RCTM_User, time=0.91, value=1.5}
            {arriveTangent=-34.1433, interpMode=RCIM_Cubic, leaveTangent=-34.1433, tangentMode=RCTM_User, time=1, value=0}
           ]}}}
        Scale Curve=10
       )
      )
     )
    )) {isSetParameters=true}
   &SolveForcesAndVelocity(Clamp Velocity=false, Limit Acceleration=false)
  ]
  rendererClasses [&NiagaraMeshRendererProperties]
  renderers [
   {rendererClass=&NiagaraMeshRendererProperties, rendererIndex=0}
   {rendererClass=&NiagaraMeshRendererProperties, rendererIndex=1}
  ]
 }}
summary {
 systemName=NS_DriedSilk_Emote_ExclamationPoint
 emitters [{emitterName=ExclamationBubble, rendererClasses [&NiagaraMeshRendererProperties]}]
 userVariables []
}
